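# Looping night ambience cue named for the Hollow Base basement POI, routed to the indoor Athena ambient sound class.
# A mixer plays a reused Starlight church interior day loop with boosted volume and doubled cue pitch; its second input is empty.
# Attenuation, spatialization and reverb send are off, so it plays as a flat 2D bed.
# A concurrency override prevents new instances once the limit is reached.
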SoundCue /Game/Sounds/Artemis/POI/Hollow_Base/Ambience_Hollow_Base_Basement_Night_Loop_Cue
root=Mixer_0
nodes [
 {id=Mixer_0, children [WavePlayer_2,null], properties {inputVolume [1.5,0.5]}}
 {
  id=WavePlayer_2
  properties {bLooping=true, soundWaveAssetPtr=/Game/Sounds/Fort_Starlight/S1/SoundWaves/AMBI_Starlight_mToD_Church_Int_Day_Loop}
 }
]
properties {
 bOverrideConcurrency=true
 maxDistance=2097152
 pitchMultiplier=2
 soundClassObject=&Fort_Ambient_Indoors_Athena
 volumeMultiplier=1.504
 attenuationOverrides {bAttenuate=false, bEnableReverbSend=false, bSpatialize=false}
 concurrencyOverrides {resolutionRule=PreventNew}
}
