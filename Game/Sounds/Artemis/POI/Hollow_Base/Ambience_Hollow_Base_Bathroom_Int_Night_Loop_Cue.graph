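# Sound cue for the looping night-time interior bathroom ambience at the Hollow Base POI.
# A single looping wave player feeds the cue at reduced volume under the indoor ambient sound class.
# Attenuation overrides turn off distance falloff, spatialization and reverb send, so it plays as a non-positional bed.
# A concurrency override uses the PreventNew rule, blocking new instances once the limit is reached.
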SoundCue /Game/Sounds/Artemis/POI/Hollow_Base/Ambience_Hollow_Base_Bathroom_Int_Night_Loop_Cue
root=WavePlayer_0
nodes [{
  id=WavePlayer_0
  properties {bLooping=true, soundWaveAssetPtr=&/Ambience_Hollow_Base_Bathroom_Int_Night_Loop}
 }]
properties {
 bOverrideConcurrency=true
 maxDistance=2097152
 soundClassObject=&Fort_Ambient_Indoors_Athena
 volumeMultiplier=0.501
 attenuationOverrides {bAttenuate=false, bEnableReverbSend=false, bSpatialize=false}
 concurrencyOverrides {resolutionRule=PreventNew}
}
