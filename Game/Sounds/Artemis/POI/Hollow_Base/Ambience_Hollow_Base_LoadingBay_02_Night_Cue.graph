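# Sound cue for the looping night ambience of the second Hollow Base loading bay point of interest.
# A quality-level node routes the first two levels to a quad mix and the last two to a stereo mix.
# Each mixer layers a metal room tone, alpine night ambience and a quieter main-cave night bed.
# It uses ambient attenuation and the outdoor ambient sound class, with overall volume reduced.
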SoundCue /Game/Sounds/Artemis/POI/Hollow_Base/Ambience_Hollow_Base_LoadingBay_02_Night_Cue
root=QualityLevel_0
nodes [
 {id=Mixer_0, children [WavePlayer_4,WavePlayer_1,WavePlayer_2], properties {inputVolume [2,2,1.2]}}
 {id=Mixer_1, children [WavePlayer_0,WavePlayer_5,WavePlayer_3], properties {inputVolume [2,2,1.2]}}
 {id=QualityLevel_0, children [Mixer_0,Mixer_0,Mixer_1,Mixer_1]}
 {id=WavePlayer_0, properties {bLooping=true, soundWaveAssetPtr=/Game/Sounds/Artemis/Biomes/Mountains/Ambience_Mountains_Alpine_Night_Stereo}}
 {id=WavePlayer_1, properties {bLooping=true, soundWaveAssetPtr=/Game/Sounds/Artemis/Biomes/Mountains/Ambience_Mountains_Alpine_Night_Quad}}
 {
  id=WavePlayer_2
  properties {bLooping=true, soundWaveAssetPtr=&/Ambience_Hollow_Base_Main_Cave_Night_Quad}
 }
 {
  id=WavePlayer_3
  properties {bLooping=true, soundWaveAssetPtr=&/Ambience_Hollow_Base_Main_Cave_Night_Stereo}
 }
 {
  id=WavePlayer_4
  properties {bLooping=true, soundWaveAssetPtr=/Game/Sounds/Artemis/RoomTones/Metal/Artemis_Interior_Medium_Metal_Room_02_Quad}
 }
 {
  id=WavePlayer_5
  properties {bLooping=true, soundWaveAssetPtr=/Game/Sounds/Artemis/RoomTones/Metal/Artemis_Interior_Medium_Metal_Room_02_Stereo}
 }
]
properties {
 attenuationSettings=&Fort_Ambient_Athena_Att
 maxDistance=2097152
 soundClassObject=&Fort_Ambient_Outdoors_Athena
 volumeMultiplier=0.601
 preEffectBusSends [{}]
}
